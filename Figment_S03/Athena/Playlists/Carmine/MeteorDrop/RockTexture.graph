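# RockTexture is a Default Lit material in the Carmine MeteorDrop playlist folder, apparently a simple rock surface for that event.
# It samples the Dracula Castle exterior rocks diffuse texture with default texture coordinates, alongside a lone constant of one.
# The graph as shown does not indicate which material inputs either node feeds, and has no parameters or further math.
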Material /Figment_S03/Athena/Playlists/Carmine/MeteorDrop/RockTexture
graph {
 properties {pixelDepthOffsetMode=PDOM_Legacy, usedShadingModels=Default Lit}
 nodes [
  Constant() {r=1}
  TextureSample(UVs=TextureCoordinate()) {texture=/Figment_S03/Packages/Fortress_Transylvania/Dracula_Castle/Textures/T_Castle_Exterior_Rocks_D}
 ]
}
